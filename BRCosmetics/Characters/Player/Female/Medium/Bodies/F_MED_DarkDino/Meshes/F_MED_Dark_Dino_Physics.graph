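# Physics asset for the female medium DarkDino player body, covering small dangling cosmetic bones: two bandana tails and two laces.
# Capsule bodies sit on each dangling bone, the bandana base and the right lower calf twist bone.
# Angle-limited constraints join both bandana tails to the base and both left and right laces to that same right calf bone, with tight twist limits.
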PhysicsAsset /BRCosmetics/Characters/Player/Female/Medium/Bodies/F_MED_DarkDino/Meshes/F_MED_Dark_Dino_Physics
bodies {
 calf_twist_02_r [{shapeType=Capsule, length=13.8432, radius=3.9567, center=<8.6422 1.9291 0.4035>, rotation=<rot 85.915 5.268 0>}]
 dyn_bandanaA_01 [{shapeType=Capsule, length=2, radius=1, center=<3 0 0>, rotation=<rot 90 0 0>}]
 dyn_bandanaB_01 [{shapeType=Capsule, length=2, radius=1, center=<3 0 0>, rotation=<rot 90 0 0>}]
 dyn_bandanaBase [{shapeType=Capsule, length=0.505, radius=0.505, rotation=<rot 0 -90 0>}]
 dyn_laces_l [{shapeType=Capsule, length=3.5, radius=1, center=<4 0 0.5>, rotation=<rot 85 0 0>}]
 dyn_laces_r [{shapeType=Capsule, length=3.5, radius=1, center=<4 0 -0.5>, rotation=<rot 85 0 0>}]
}
constraints [
 {bone1Name=dyn_laces_l, bone2Name=calf_twist_02_r, swing1LimitDegrees=45, swing2LimitDegrees=90, twistLimitDegrees=10}
 {bone1Name=dyn_laces_r, bone2Name=calf_twist_02_r, swing1LimitDegrees=45, swing2LimitDegrees=90, twistLimitDegrees=10}
 {bone1Name=dyn_bandanaA_01, bone2Name=dyn_bandanaBase, swing1LimitDegrees=45, swing2LimitDegrees=45, twistLimitDegrees=5}
 {bone1Name=dyn_bandanaB_01, bone2Name=dyn_bandanaBase, swing1LimitDegrees=45, swing2LimitDegrees=45, twistLimitDegrees=5}
]
